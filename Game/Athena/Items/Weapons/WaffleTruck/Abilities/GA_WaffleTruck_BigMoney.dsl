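; Gameplay ability for the WaffleTruck Big Money weapon that drops cash wad pickups when its wielder eliminates a hostile player.
; Event activation confirms the damage came from the source weapon and was a fatal hit on a hostile player pawn.
; A re-armed timer makes the server toss one pickup above the target per tick, playing a kill gameplay cue between drops.
; The ability ends at the spawn count, clearing the timer.
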(event Custom|ExecuteGameplayCue_Kill
  ;[1 1 1] Fx
  (bind _abilitycomponent (Utilities|GetValidOwningAbilityComponent false)
    (:true
      (Ability|ExecuteGameplayCueOnOwner (Variables|BigMoney|Default|GetGameplayCue_BigMoney_Kill) (GameplayEffects|MakeEffectContext _abilitycomponent)))))

(event Custom|BeginSpawn
  ;[1 1 1] Spawn Loop
  (bind _returnvalue (Utilities|Time|SetTimerbyEvent (delegate Custom|Spawn) (Variables|BigMoney|Default|GetSpawnTime)))
  (Variables|BigMoney|Dynamic|SetSpawnTimerHandle _returnvalue))

(event Custom|Spawn
  (bind (_returnvalue_x _returnvalue_y _returnvalue_z) (Math|Vector|BreakVector (Transformation|GetActorLocation (Variables|BigMoney|Dynamic|GetTarget))))
  ;[1 1 1] Spawn Loop
  (Utilities|IsServer
    (:true
      (Inventory|SpawnPickupInWorld (Variables|BigMoney|Default|GetItemDef_CashWad) 1 (Math|Vector|MakeVector _returnvalue_x _returnvalue_y (+ _returnvalue_z (Variables|Default|GetSpawnZOffset))) (Transformation|GetActorUpVector (Variables|BigMoney|Dynamic|GetTarget)) 0 true true false -1 "Tossed" "PlayerElimination" (Utilities|Casting|CastToFortPlayerController (Pawn|GetController (CallFunction|GetAvatarActorandAbilityComponent))) true 0))
    (:false
      _))
  (bind _bspawncompleted (CallFunction|IncrementCashDropandCheckforComplete))
  (if _bspawncompleted
    (Ability|EndAbility)))

(event Ability|EventOnEndAbility
  (params (bWasCancelled :type Boolean))
  ;[1 1 1] End
  (Utilities|Time|ClearandInvalidateTimerbyHandle (Variables|BigMoney|Dynamic|GetSpawnTimerHandle)))

(event Ability|EventActivateAbilityFromEvent
  (params (EventData :type Gameplay%20Event%20Data%20Structure%20%28by%20ref%29))
  (bind (_eventtag _instigator _target _optionalobject _optionalobject2 _contexthandle) (Utilities|Struct|BreakGameplayEventData EventData))
  ;[1 1 1] Activate
  (Utilities|DamageIsfromSourceWeapon EventData
    (:True
      (Utilities|GetValidOwningAbilityComponent true
        (:true
          (bind _asfort_player_pawn (Utilities|GetValidTargetasFortPlayerPawn _target)
            (:Valid
              (bind _output_get (Variables|BigMoney|Dynamic|SetTarget _asfort_player_pawn))
              (Utilities|ExecuteonFatalHitAgainstHostile _output_get _instigator _contexthandle
                (:Exec
                  (Variables|BigMoney|Dynamic|SetCurrentCount 0)
                  (CallFunction|BeginSpawn))))))))))

(fn GetAvatarActorAndAbilityComponent
  (return (CallFunction|GetAvatarasFortPlayerPawn) (Ability|GetAbilitySystemComponent (CallFunction|GetAvatarasFortPlayerPawn)) (and (Utilities|IsValid (CallFunction|GetAvatarasFortPlayerPawn) :!class /Script/Engine.KismetSystemLibrary) (Utilities|IsValid (Ability|GetAbilitySystemComponent (CallFunction|GetAvatarasFortPlayerPawn)) :!class /Script/Engine.KismetSystemLibrary))))

(fn IncrementCashDropAndCheckForComplete
  (bind _result (Math|Integer|IncrementInt (Variables|BigMoney|Dynamic|GetCurrentCount))
    (:%20%20
      (bind _output_get (Variables|BigMoney|Dynamic|SetCurrentCount _result))
      (if (< _output_get (Variables|BigMoney|Default|GetSpawnCount))
        (CallFunction|ExecuteGameplayCueKill)
        (CallFunction|BeginSpawn)
        (return false)
        (else
          (return true))))))

(fn IsDamageFromSourceWeapon
  (params (GameplayEventData :type Gameplay%20Event%20Data%20Structure%20%28by%20ref%29))
  (bind (_eventtag _instigator _target _optionalobject _optionalobject2 _contexthandle) (Utilities|Struct|BreakGameplayEventData GameplayEventData))
  (bind _avatar (CallFunction|GetAvatarActorandAbilityComponent))
  (Utilities|IsValid _avatar
    (:Is%20Valid
      (return (Utilities|Equal(Object) (Variables|/Script/FortniteGame.FortPawn|GetCurrentWeapon _avatar) (Ability|EffectContext|GetSourceObject _contexthandle))))
    (:Is%20Not%20Valid
      (return false))))

(fn OwnerAndTargetAreHostile
  (params (ActorA :type Actor%20Object%20Reference) (ActorB :type Actor%20Object%20Reference))
  (return (Utilities|Enum|Equal(Enum) (Team|GetTeamAffiliation ActorA ActorB) "Hostile")))

(fn TargetIsHostileAndIsFatalShot
  (params (Instigator :type Actor%20Object%20Reference) (Target :type Actor%20Object%20Reference) (EffectContext :type Gameplay%20Effect%20Context%20Handle%20Structure))
  (return (and (CallFunction|OwnerandTargetAreHostile Instigator Target) (Ability|EffectContext|IsFatalHit EffectContext))))
