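; Gameplay ability for the WaffleTruck weapon set that grants a Run N Gun buff to its owner, covering accuracy and movement speed by name.
; On activation it gets a valid owning ability component, applies the RunNGunAccuracy and RunNGunSpeed gameplay effects to it, then ends.
; EndAbility runs only on the valid-component branch.
; A helper function returns the avatar player pawn, its ability system component and a combined validity flag.
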(event Ability|EventActivateAbility
  (bind _abilitycomponent (Utilities|GetValidOwningAbilityComponent true)
    (:true
      (GameplayEffects|ApplyGameplayEffectToSelf _abilitycomponent (Variables|Default|GetRunNGunAccuracy) 0.0 (GameplayEffects|MakeEffectContext _abilitycomponent))
      (GameplayEffects|ApplyGameplayEffectToSelf _abilitycomponent (Variables|Default|GetRunNGunSpeed) 0.0 (GameplayEffects|MakeEffectContext _abilitycomponent))
      (Ability|EndAbility))))

(fn GetAvatarActorAndAbilityComponent
  (return (CallFunction|GetAvatarasFortPlayerPawn) (Ability|GetAbilitySystemComponent (CallFunction|GetAvatarasFortPlayerPawn)) (and (Utilities|IsValid (CallFunction|GetAvatarasFortPlayerPawn) :!class /Script/Engine.KismetSystemLibrary) (Utilities|IsValid (Ability|GetAbilitySystemComponent (CallFunction|GetAvatarasFortPlayerPawn)) :!class /Script/Engine.KismetSystemLibrary))))
